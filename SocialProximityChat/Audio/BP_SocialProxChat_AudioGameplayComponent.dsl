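; Audio gameplay component for social proximity chat that applies a concurrency set to a Fortnite player pawn's voice chat audio.
; On BeginPlay, where cosmetic events can execute, it casts its owner to FortPlayerPawnAthena and binds to that pawn's OnVoiceChatComponentCreated dispatcher.
; When the voice component is created and valid, it assigns the VoiceConcurrencySet variable as that audio component's ConcurrencySet.
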(event EventBeginPlay
  (bind _returnvalue (Components|GetOwner))
  (Networking|CanExecuteCosmeticEvents
    (:True
      (Utilities|IsValid _returnvalue
        (:Is%20Valid
          (bind _asfort_player_pawn_athena (Utilities|Casting|CastToFortPlayerPawnAthena _returnvalue))
          (EventDispatchers|BindEventtoOnVoiceChatComponentCreated _asfort_player_pawn_athena (EventDispatchers|CreateEvent :!event Custom|OnVoiceComponentCreated)))))))

(event Custom|OnVoiceComponentCreated
  (params (FortPlayerPawnAthena :type Fort%20Player%20Pawn%20Athena%20Object%20Reference) (VoiceChatComponent :type Audio%20Component%20Object%20Reference))
  (Utilities|IsValid VoiceChatComponent
    (:Is%20Valid
      (Variables|/Script/Engine.AudioComponent|SetConcurrencySet (Variables|Default|GetVoiceConcurrencySet) VoiceChatComponent))))
