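; Blueprint renderer for social proximity voice chat that supplies and configures the sound used to play a speaker's voice.
; CreateProximityChatVoiceSound returns the VoiceChatMetaSound variable, ignoring the passed audio bus and attenuation settings.
; When proximity voice starts, a valid speaking audio component receives the audio bus through its "AudioBus" object parameter, routing it into the MetaSound.
; BeginPlay and Tick are empty.
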(event EventBeginPlay)

(event EventTick
  (params (DeltaSeconds :type Float%20%28single-precision%29)))

(event EventHandleProximityChatVoiceStarted
  (params (SpeakingAudioComponent :type Audio%20Component%20Object%20Reference) (AudioBus :type Audio%20Bus%20Object%20Reference) (AttenuationSettings :type Sound%20Attenuation%20Object%20Reference))
  (Utilities|IsValid SpeakingAudioComponent
    (:Is%20Valid
      (Audio|Parameter|SetObjectParameter SpeakingAudioComponent "AudioBus" AudioBus))))

(fn CreateProximityChatVoiceSound
  (params (AudioBus :type Audio%20Bus%20Object%20Reference) (AttenuationSettings :type Sound%20Attenuation%20Object%20Reference))
  (return (Variables|Default|GetVoiceChatMetaSound)))
