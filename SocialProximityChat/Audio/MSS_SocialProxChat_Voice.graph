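# Looping mono MetaSound source that plays proximity chat voice from an audio bus supplied as an input.
# The bus signal is boosted by a fixed gain, squashed by a high-ratio compressor, then trimmed before output, evening out voice levels.
# The source uses the social proximity chat attenuation and voice concurrency assets, and the Fortnite proximity chat sound class.
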MetaSoundSource /SocialProximityChat/Audio/MSS_SocialProxChat_Voice
inputs [
 AudioBus {type=AudioBusAsset}
 UE.Source.OnPlay {type=Trigger}
]
interfaces [UE.OutputFormat.Mono:1.0, UE.Source:1.1]
outputs ["UE.OutputFormat.Mono.Audio:0" {type=Audio}]
pages [Default {
  input_defaults {AudioBus=null, UE.Source.OnPlay=null}
  nodes [
   AudioBusReader1_0 = "UE.Audio Bus Reader (1)"(Audio Bus=^AudioBus)
   DecibelstoLinearGain_1 = UE.Decibels to Linear Gain:Float() {also {defaults {Decibels=12}}}
   DecibelstoLinearGain_0 = UE.Decibels to Linear Gain:Float() {also {defaults {Decibels=-6}}}
   Multiply_1 = UE.Multiply:Audio by Float(
    PrimaryOperand=UE.Compressor:Audio(Audio=UE.Multiply:Audio by Float(PrimaryOperand=@AudioBusReader1_0.Out 0, AdditionalOperands=@DecibelstoLinearGain_1.Linear Gain).Out) {also {defaults {Envelope Mode=2, Ratio=10}}}.Audio
    AdditionalOperands=@DecibelstoLinearGain_0.Linear Gain
   )
  ]
  links [{dst="@UE.OutputFormat.Mono.Audio:0", src=Multiply_1.Out}]
 }]
properties {
 assetImportData=&self:AssetImportData
 attenuationSettings=&/ATT_SocialProxChat
 bLooping=true
 duration=10000
 maxDistance=20000
 sampleRate=48000
 soundClassObject=/Game/Sounds/FortSoundClassesAndModes/Fort_ProximityChat
 concurrencySet [&/CONC_SocialProxChat_Voice]
}
